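; Parent Blueprint for player building pieces, apparently walls, that gives each placed actor a random mesh variant.
; On BeginPlay, only the authority acts: if alternate meshes exist, it inserts the current mesh into that array and sets a random entry, relying on mesh replication for clients.
; The random range's inclusive upper bound is the array length, so it can land past the last entry.
; The RandomChance RepNotify and construction script are empty.
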(event EventBeginPlay
  (bind _staticmeshalternatearray (Variables|WallVariants|GetStaticMeshAlternateArray))
  (Networking|SwitchHasAuthority
    (:Authority
      (if (> (Utilities|Array|Length _staticmeshalternatearray) 0)
        (Utilities|Array|Insert _staticmeshalternatearray (Variables|Mesh|GetStaticMesh))
        ;[1 1 1] static mesh is replicated
        (Variables|Mesh|SetStaticMesh (Utilities|Array|Get(acopy) _staticmeshalternatearray (Math|Random|RandomIntegerinRange 0 (Utilities|Array|Length (Variables|WallVariants|GetStaticMeshAlternateArray)))))))))

(fn OnRep_RandomChance)

(fn ConstructionScript)
